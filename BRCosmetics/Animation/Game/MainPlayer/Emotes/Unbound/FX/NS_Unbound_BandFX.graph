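# Niagara effect for the band in the main player's Unbound emote, made of glowing sparks and a warm light.
# Band_Motes continuously spawns short-lived orange-yellow sprites on a skeletal mesh, moved by curl noise and drag while shrinking and fading.
# Light burst-spawns a roughly three-second sprite with a light renderer, pulsing faster through a curve-driven waveform, then flaring brighter and stretching wide at the end of its life.
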NiagaraSystem /BRCosmetics/Animation/Game/MainPlayer/Emotes/Unbound/FX/NS_Unbound_BandFX
dependencies {
 usedDataInterfaces [
  &NiagaraDataInterfaceCurve
  &NiagaraDataInterfaceColorCurve
  &NiagaraDataInterfaceVector2DCurve
  &NiagaraDataInterfaceSkeletalMesh
 ]
 usedDynamicInputs []
 usedModules [
  &self:SystemSpawnScript.SystemScriptSource.SystemScriptGraph.NiagaraNodeAssignment_0.SetVariables_#1
  &SystemState
 ]
}
emitters {
 Band_Motes {
  emitterUpdateScript [
   &EmitterState()
   &SpawnRate(SpawnRate=72, Use Spawn Probability=false)
  ]
  particleSpawnScript [
   &InitializeParticle(
    Color Maximum=<rgba 15 11.8311 0.9375 1>
    Color Minimum=<rgba 15 2.1948 0 1>
    Color Mode=::NewEnumerator2
    Lifetime Max=0.4
    Lifetime Min=0.2
    Lifetime Mode=::NewEnumerator1
    Sprite Size Mode=::NewEnumerator1
    Uniform Sprite Size Max=1.4
    Uniform Sprite Size Min=1
    UsePositionOffset=false
    Write Material Random=true
   )
   &SkeletalMeshLocation(Bone Sampling Mode=::NewEnumerator0, MaskSpawnGroup=false, OverrideRandomSeed=false, Skeletal Mesh=$System.Skeletal Mesh)
   &AddVelocity(
    Module.Use Falloff Distance=false
    Override Seed=false
    Velocity Mode=::NewEnumerator1
    Velocity Speed=&RandomRangeFloat(Minimum=10, Maximum=25)
   )
  ]
  particleUpdateScript [
   &CurlNoiseForce(Bias Noise=false, Noise Strength=75, Pan Noise=false)
   &Drag(
    Drag=&FloatFromCurve(
     FloatCurve=DataInterface() {propertyValues {Curve {keys [<at 0 0 RCIM_Cubic>,<at 0.1 1 RCIM_Cubic>]}}}
     Scale Curve=&RandomRangeFloat(Minimum=0.8, Maximum=1.2)
    )
    Use Linear Drag=true
    Use Rotational Drag=true
   )
   &InheritSourceMovement(
    Applied Position Scale=&VectorFromFloat(Value=&FloatFromCurve(FloatCurve=DataInterface() {propertyValues {Curve {keys [<at 0 0 RCIM_Cubic>,<at 0.1 1 RCIM_Cubic>]}}}, Scale Curve=0.5))
    ApplyForce=false
    ApplyPosition=true
    ApplyPositionFalloff=false
   ) {enabled=false}
   &SolveForcesAndVelocity(Clamp Velocity=false, Limit Acceleration=false)
   &ParticleState()
   &ScaleSpriteSizeBySpeed(Max Scale Factor=<0.5 2>, Velocity Threshold=200)
   &ScaleSpriteSize(Uniform Curve Sprite Scale=DataInterface() {propertyValues {Curve {keys [<at 0 1 RCIM_Cubic>,<at 1 0 RCIM_Cubic>]}}})
   &ScaleColor(
    Linear Color Curve=DataInterface() {propertyValues {
      AlphaCurve {keys [<at 0 1>,<at 1 0>]}
      BlueCurve {keys [<at 0 1>,<at 0.8 0>]}
      GreenCurve {keys [<at 0 1>,<at 0.8 0.0299>]}
      RedCurve {keys [<at 0 1>,<at 0.8 1>]}
     }}
    Scale Mode=::NewEnumerator2
   )
  ]
  rendererClasses [&NiagaraSpriteRendererProperties]
  renderers [{rendererClass=&NiagaraSpriteRendererProperties}]
 }
 Light {
  emitterUpdateScript [
   &EmitterState()
   &SpawnBurst_Instantaneous(Use Spawn Probability=false, UseLoopCountLimit=false)
  ]
  particleSpawnScript [&InitializeParticle(
    Color=<rgba 5 3.5984 0.9115 0.2>
    Lifetime=3.04
    Sprite Size=<25 25>
    Sprite Size Mode=::NewEnumerator4
    UsePositionOffset=false
    Write Material Random=true
   )]
  particleUpdateScript [
   &ParticleState()
   &ScaleSpriteSize(
    Non-Uniform Curve Sprite Scale=DataInterface() {propertyValues {
      XCurve {keys [
        {tangentMode=RCTM_User, time=0, value=1}
        {arriveTangent=0.0059, arriveTangentWeight=0.2333, interpMode=RCIM_Cubic, leaveTangent=0.0059, tangentMode=RCTM_User, time=0.85, value=1}
        {arriveTangent=0.1338, arriveTangentWeight=0.1, interpMode=RCIM_Cubic, leaveTangent=0.1338, tangentMode=RCTM_User, time=1, value=1.6}
       ]}
      YCurve {keys [
        {tangentMode=RCTM_User, time=0, value=1}
        {arriveTangent=0.0059, arriveTangentWeight=0.2333, interpMode=RCIM_Cubic, leaveTangent=0.0059, tangentMode=RCTM_User, time=0.85, value=1}
        {arriveTangent=0.1338, arriveTangentWeight=0.1, interpMode=RCIM_Cubic, leaveTangent=0.1338, tangentMode=RCTM_User, time=1, value=0.9}
       ]}
     }}
    Scale Sprite Size Mode=::NewEnumerator3
   )
   &ScaleSpriteSize(
    Scale Sprite Size Mode=::NewEnumerator0
    Uniform Scale Factor=&Waveform(
     "Waveform [1]"=::NewEnumerator0
     "[1] Frequency"=&FloatFromCurve(
      FloatCurve=DataInterface() {propertyValues {Curve {keys [<at 0 0 RCIM_Cubic>, {arriveTangent=3.087, interpMode=RCIM_Cubic, leaveTangent=3.087, tangentMode=RCTM_User, time=1, value=1}]}}}
      Scale Curve=4
     )
     "[1] Amplitude Min/Max"=<1.1 1.3>
    )
   ) {module=ScaleSpriteSize001}
   &ScaleColor(
    Scale Alpha=&FloatFromCurve(FloatCurve=DataInterface() {propertyValues {Curve {keys [{arriveTangent=1.1111, interpMode=RCIM_Cubic, time=0.7, value=1}, <at 1 2>]}}})
    Scale RGB=&VectorFromFloat(Value=&FloatFromCurve(FloatCurve=DataInterface() {propertyValues {Curve {keys [
          {arriveTangent=11.612, interpMode=RCIM_Cubic, leaveTangent=11.612, leaveTangentWeight=0.2333, tangentMode=RCTM_User, time=0, value=1}
          {
           arriveTangent=0.0237
           arriveTangentWeight=0.0083
           interpMode=RCIM_Cubic
           leaveTangent=0.0237
           leaveTangentWeight=0.225
           tangentMode=RCTM_User
           time=0.025
           value=2.2
          }
          {arriveTangent=-0.0058, arriveTangentWeight=0.025, leaveTangent=-0.0058, leaveTangentWeight=0.2, time=0.1, value=1}
          {arriveTangent=0.0319, interpMode=RCIM_Cubic, leaveTangent=0.0319, tangentMode=RCTM_User, time=0.7, value=1}
          {arriveTangent=6.0453, interpMode=RCIM_Cubic, leaveTangent=6.0453, tangentMode=RCTM_User, time=1, value=3}
         ]}}}))
    ScaleA=true
    ScaleRGB=false
   )
   &FN_Light_Attributes(
    Color=$Particles.Color
    Create Color Attribute=true
    Create Exponent Attribute=true
    Create Radius Attribute=true
    Radius=100
    Scale Color=&FloatFromCurve(FloatCurve=DataInterface() {propertyValues {Curve {keys [
         {arriveTangent=3.9605, interpMode=RCIM_Cubic, leaveTangent=3.9605, tangentMode=RCTM_User, time=0, value=0}
         {arriveTangent=11.0938, interpMode=RCIM_Cubic, leaveTangent=11.0938, tangentMode=RCTM_User, time=0.1, value=0.6}
         {arriveTangent=6.2676, interpMode=RCIM_Cubic, leaveTangent=6.2676, tangentMode=RCTM_User, time=0.1, value=0.6}
         {arriveTangent=0.041, interpMode=RCIM_Cubic, leaveTangent=0.041, tangentMode=RCTM_User, time=0.3, value=1}
         {arriveTangent=0.0399, interpMode=RCIM_Cubic, leaveTangent=0.0399, tangentMode=RCTM_User, time=0.85, value=1}
         <at 1 3>
        ]}}})
   )
   &CameraOffset()
  ]
  rendererClasses [&NiagaraSpriteRendererProperties, &NiagaraLightRendererProperties]
  renderers [
   {rendererClass=&NiagaraSpriteRendererProperties}
   {rendererClass=&NiagaraLightRendererProperties, rendererIndex=1}
  ]
 }
}
summary {
 systemName=NS_Unbound_BandFX
 emitters [
  {emitterName=Band_Motes, rendererClasses [&NiagaraSpriteRendererProperties]}
  {emitterName=Light, rendererClasses [&NiagaraSpriteRendererProperties, &NiagaraLightRendererProperties]}
 ]
 userVariables []
}
