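# Niagara system for a brief warm golden flash, named as the band start effect for the Unbound main-player emote.
# Its single Light emitter bursts a quarter-second particle drawn by both sprite and light renderers.
# Curves make the sprite overshoot in size then settle while its alpha and light color ramp up, with no fade-out before it expires.
# A negative camera offset shifts the particle along the view axis.
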NiagaraSystem /BRCosmetics/Animation/Game/MainPlayer/Emotes/Unbound/FX/NS_Unbound_BandFXStart
dependencies {
 usedDataInterfaces [&NiagaraDataInterfaceCurve, &NiagaraDataInterfaceColorCurve, &NiagaraDataInterfaceSkeletalMesh]
 usedDynamicInputs []
 usedModules [
  &self:SystemSpawnScript.SystemScriptSource.SystemScriptGraph.NiagaraNodeAssignment_0.SetVariables_#1
  &SystemState
 ]
}
emitters {Light {
  emitterUpdateScript [
   &EmitterState()
   &SpawnBurst_Instantaneous(Use Spawn Probability=false, UseLoopCountLimit=false)
  ]
  particleSpawnScript [&InitializeParticle(
    Color=<rgba 5 3.5984 0.9115 0.2>
    Lifetime=0.25
    Sprite Rotation Mode=::NewEnumerator2
    Sprite Size Mode=::NewEnumerator3
    Uniform Sprite Size=25
    UsePositionOffset=false
    Write Material Random=true
   )]
  particleUpdateScript [
   &ParticleState()
   &ScaleSpriteSize(Uniform Curve Sprite Scale=DataInterface() {propertyValues {Curve {keys [
        {arriveTangent=3.6507, interpMode=RCIM_Cubic, leaveTangent=3.6507, tangentMode=RCTM_User, time=0, value=0.5}
        {arriveTangent=0.0237, interpMode=RCIM_Cubic, leaveTangent=0.0237, tangentMode=RCTM_User, time=0.4, value=1.4}
        {arriveTangent=-1.3483, interpMode=RCIM_Cubic, leaveTangent=-1.3483, tangentMode=RCTM_User, time=1, value=1}
       ]}}})
   &ScaleColor(
    Scale Alpha=&FloatFromCurve(FloatCurve=DataInterface() {propertyValues {Curve {keys [
         {arriveTangent=3.1732, interpMode=RCIM_Cubic, leaveTangent=3.1732, leaveTangentWeight=0.3333, tangentMode=RCTM_User, time=0, value=0}
         {arriveTangent=-0.0058, arriveTangentWeight=0.025, leaveTangent=-0.0058, leaveTangentWeight=0.3, time=1, value=1}
        ]}}})
    Scale RGB=&VectorFromFloat(Value=&FloatFromCurve(FloatCurve=DataInterface() {propertyValues {Curve {keys [
          {arriveTangent=11.612, interpMode=RCIM_Cubic, leaveTangent=11.612, leaveTangentWeight=0.2333, tangentMode=RCTM_User, time=0, value=1}
          {
           arriveTangent=0.0237
           arriveTangentWeight=0.0083
           interpMode=RCIM_Cubic
           leaveTangent=0.0237
           leaveTangentWeight=0.225
           tangentMode=RCTM_User
           time=0.025
           value=2.2
          }
          {arriveTangent=-0.0058, arriveTangentWeight=0.025, leaveTangent=-0.0058, leaveTangentWeight=0.2, time=0.1, value=1}
          {arriveTangent=0.0319, interpMode=RCIM_Cubic, leaveTangent=0.0319, tangentMode=RCTM_User, time=0.7, value=1}
          {arriveTangent=6.0453, interpMode=RCIM_Cubic, leaveTangent=6.0453, tangentMode=RCTM_User, time=1, value=3}
         ]}}}))
    ScaleA=true
    ScaleRGB=false
   )
   &FN_Light_Attributes(
    Color=$Particles.Color
    Create Color Attribute=true
    Create Exponent Attribute=true
    Create Radius Attribute=true
    Radius=100
    Scale Color=&FloatFromCurve(FloatCurve=DataInterface() {propertyValues {Curve {keys [
         {arriveTangent=7.6639, interpMode=RCIM_Cubic, leaveTangent=7.6639, tangentMode=RCTM_User, time=0, value=0}
         {arriveTangent=-0.0028, interpMode=RCIM_Cubic, leaveTangent=-0.0028, tangentMode=RCTM_User, time=0.3, value=1}
         {arriveTangent=-0.0108, interpMode=RCIM_Cubic, leaveTangent=-0.0108, tangentMode=RCTM_User, time=1, value=1}
        ]}}})
   )
   &CameraOffset(Camera Offset Amount=-5)
  ]
  rendererClasses [&NiagaraSpriteRendererProperties, &NiagaraLightRendererProperties]
  renderers [
   {rendererClass=&NiagaraSpriteRendererProperties}
   {rendererClass=&NiagaraLightRendererProperties, rendererIndex=1}
  ]
 }}
summary {
 systemName=NS_Unbound_BandFXStart
 emitters [{emitterName=Light, rendererClasses [&NiagaraSpriteRendererProperties, &NiagaraLightRendererProperties]}]
 userVariables []
}
